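# Translucent, two-sided unlit master material for melee weapon trails on particle sprites, beams and ribbons, with optional refraction.
# Panned noise, smoke and caustic textures, warped by noise UVs, blend tinted colors and form opacity inside soft UV masks.
# Dynamic parameters drive time offset, desaturation, erosion and distortion, while a separate panned normal/texture branch adds waving.
# Mobile drops UV warp, waving and depth/camera fades, and lower quality levels flatten refraction.
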Material /Game/Effects/Fort_Effects/Materials/Melee/M_Melee_Trail_Distortion_Master
graph {
 properties {
  bUsedWithBeamTrails=true
  bUsedWithNiagaraRibbons=true
  bUsedWithNiagaraSprites=true
  bUsedWithParticleSprites=true
  blendMode=BLEND_Translucent
  pixelDepthOffsetMode=PDOM_Legacy
  refractionMethod=RM_IndexOfRefraction
  shadingModel=MSM_Unlit
  twoSided=true
  usedShadingModels=Unlit
 }
 nodes [
  n1 = Constant2Vector() {r=0.2}
  n2 = AppendVector(A=ScalarParameter() {parameterName=PanSpeed}, B=Constant())
  n3 = TextureCoordinate() {coordinateIndex=1, uTiling=2}
  n4 = ComponentMask(None=TextureSampleParameter2D(UVs=Panner(Coordinate=@n3) {speedX=-0.4}) {parameterName=NoiseTexture, texture=&T_NoisePack_01}.RGB) {g=true, r=true}
  n5 = ScalarParameter() {defaultValue=8, parameterName=UVdistortAmount}
  n6 = Add(
   A=ShadingPathSwitch(Default=Divide(A=@n4, B=@n5) {constB=8}, Mobile=Constant()) {desc=Migrated from FeatureLevelSwitch}
   B=Multiply(A=TextureCoordinate() {coordinateIndex=1, vTiling=0.7}, B=ScalarParameter() {defaultValue=1, parameterName=NoiseTile})
  )
  n7 = Add(A=Time(), B=DynamicParameter() {defaultValue=<rgba 0 1 0.4 0>, paramNames [TimeOffset,Curl_Scale,Desa,Erode]}.TimeOffset)
  n8 = TextureSample(UVs=Panner(Coordinate=@n6, Time=@n7, Speed=Add(A=@n1, B=@n2) {constB=0.2}) {speedX=0.3}) {texture=&T_NoisePack_01}
  n9 = Desaturation(None=@n8.RGB, Fraction=DynamicParameter() {defaultValue=<rgba 0 1 0.4 0>, paramNames [TimeOffset,Curl_Scale,Desa,Erode]}.Desa)
  n10 = Multiply(A=VectorParameter() {parameterName=HighColor, defaultValue=<rgba 0.0479 0.1589 1 1>}.RGB, B=@n9)
  n11 = Multiply(A=VectorParameter() {parameterName=LowColor, defaultValue=<rgba 0.312 0.4655 2 1>}.RGB, B=@n9)
  n12 = MaterialFunctionCall(Float2=@n6) {materialFunction=&BreakOutFloat2Components}
  n13 = Add(A=@n12.G) {constB=0.5}
  n14 = TextureSample(UVs=Panner(Coordinate=AppendVector(A=@n12.R, B=@n13), Time=@n7, Speed=@n2)) {texture=&T_Inky_Smoke_Tile}
  n15 = Constant2Vector() {r=0.3}
  n16 = TextureSample(UVs=Panner(Coordinate=@n6, Time=@n7, Speed=Add(A=@n15, B=@n2) {constB=0.3}) {speedX=0.2}) {texture=&T_ShieldCaustics}
  n17 = TextureCoordinate()
  n18 = SphereMask(
   A=Power(Base=ComponentMask(None=@n17) {r=true}, Exp=ScalarParameter() {defaultValue=0.5, parameterName=OpatcityMaskUbias}) {constExponent=0.5}
   B=Constant() {r=0.5}
  ) {attenuationRadius=0.5, hardnessPercent=0}
  n19 = TextureCoordinate()
  n20 = Saturate(None=SphereMask(A=ComponentMask(None=@n19) {g=true}, B=Constant() {r=0.5}) {attenuationRadius=0.5, hardnessPercent=0})
  n21 = Multiply(A=Multiply(A=@n18, B=@n18), B=Multiply(A=@n20, B=@n20))
  n22 = Add(A=LinearInterpolate(A=ScalarParameter() {parameterName=Sorften}, Alpha=@n16.R), B=Multiply(A=@n16.R, B=@n21))
  n23 = Multiply(A=Add(A=LinearInterpolate(Alpha=@n14.R) {constB=3}, B=LinearInterpolate(Alpha=@n22) {constA=0.2, constB=3}), B=@n21)
  n24 = ParticleColor()
  n25 = Multiply(
   A=LinearInterpolate(
    A=VectorParameter() {parameterName=Base Color, defaultValue=<rgba 0.0016 0 0.0052 1>}.RGB
    B=LinearInterpolate(A=@n10, B=@n11, Alpha=@n8.G)
    Alpha=@n23
   )
   B=@n24.RGB
  )
  n26 = Clamp(None=@n25)
  n27 = ScalarParameter() {defaultValue=1, parameterName=Mobile HDR Intensity}
  MaterialFunctionCall(Emissive=ShadingPathSwitch(
    Default=MaterialFunctionCall(LDR Input=@n26, HDR Tint=@n25) {materialFunction=&DeriveHDRfromLDR}.RemappedRange
    Mobile=Multiply(A=@n27, B=@n25)
   ) {desc=Migrated from FeatureLevelSwitch}) {materialFunction=&MF_FX_EmissiveAdjustments}
  n28 = Subtract(A=@n23, B=DynamicParameter() {defaultValue=<rgba 0 1 0.4 0>, paramNames [TimeOffset,Curl_Scale,Desa,Erode]}.Erode)
  n29 = Multiply(A=@n24.A, B=Clamp(None=@n28, Max=@n28))
  n30 = ScalarParameter() {parameterName=CameraFadeDistance}
  ShadingPathSwitch(
   Default=Multiply(
    A=DepthFade(Opacity=@n29)
    B=MaterialFunctionCall(Distance=@n30) {materialFunction=&Near_Camera_Fade}.Result
   )
   Mobile=@n29
  ) {desc=Migrated from FeatureLevelSwitch}
  n31 = Panner(
   Coordinate=TextureCoordinate() {coordinateIndex=1, uTiling=0.3, vTiling=0.3}
   Time=Add(A=Time(), B=DynamicParameter() {defaultValue=<rgba 0 1 0.4 0>, paramNames [TimeOffset,Curl_Scale,Desa,Erode]}.TimeOffset)
   Speed=AppendVector(A=ScalarParameter() {parameterName=WavingSpeed}, B=Constant())
  )
  n32 = AppendVector(
   A=ComponentMask(None=TextureSample(UVs=@n31) {samplerType=SAMPLERTYPE_Normal, texture=&T_Weap_Distort_N_02}.RGB) {g=true, r=true}
   B=TextureSample(UVs=@n31) {texture=&T_Smooth_tile}.R
  )
  n33 = ScalarParameter() {parameterName=WavingAmount}
  n34 = Multiply(A=ComponentMask(None=TextureCoordinate() {coordinateIndex=1, uTiling=0.3, vTiling=0.3}) {r=true}) {constB=1.5}
  ShadingPathSwitch(Default=Multiply(A=Multiply(A=@n32, B=@n33) {constB=1500}, B=Saturate(None=@n34)), Mobile=Constant()) {desc=Migrated from FeatureLevelSwitch}
  n35 = MaterialFunctionCall(Value=Multiply(A=MaterialFunctionCall() {materialFunction=&LinearGradient}.VGradient) {constB=0.5}) {materialFunction=&LinearSine}
  n36 = LinearInterpolate(
   B=DynamicParameter() {parameterIndex=1, defaultValue=<rgba 0 0 0 0>, paramNames [Distortion Amount, 0, 0, 0]}.Distortion Amount
   Alpha=Power(Base=Multiply(A=@n35.Linear Sine)) {constExponent=16}
  ) {constA=1, constB=1.05}
  n37 = Constant() {r=1}
  StaticSwitchParameter(True=QualitySwitch(Default=@n36, Low=@n37, High=@n36, Medium=@n37), False=@n37) {parameterName=UseRefraction}
 ]
}
