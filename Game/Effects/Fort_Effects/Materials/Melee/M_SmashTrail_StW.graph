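# Unlit, two-sided translucent beam trail material for a melee smash effect, tinted by particle color through MF_FX_EmissiveAdjustments.
# T_SmashTrail's R drives a contrast lerp, blended via LerpDetails with a version multiplied by its panning G detail, tiled on UV1 by TileDetail.
# The result is multiplied by texture B and particle alpha.
# The contrast lerp uses Contrast_Neg for both ends, so R has no effect and Contrast_Pos is unused.
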Material /Game/Effects/Fort_Effects/Materials/Melee/M_SmashTrail_StW
graph {
 properties {
  bEnableResponsiveAA=true
  bUsedWithBeamTrails=true
  blendMode=BLEND_Translucent
  pixelDepthOffsetMode=PDOM_Legacy
  shadingModel=MSM_Unlit
  twoSided=true
  usedShadingModels=Unlit
 }
 nodes [
  n1 = ParticleColor()
  MaterialFunctionCall(Emissive=@n1.RGB) {materialFunction=&MF_FX_EmissiveAdjustments}
  n2 = DynamicParameter() {defaultValue=<rgba 0 0 0 1>, paramNames [Contrast_Neg,Contrast_Pos,LerpDetails,TileDetail]}
  n3 = TextureSample(UVs=TextureCoordinate()) {texture=/Game/Effects/Fort_Effects/Textures/Melee/T_SmashTrail}
  n4 = LinearInterpolate(A=@n2.Contrast_Neg, B=@n2.Contrast_Neg, Alpha=@n3.R)
  n5 = TextureCoordinate() {coordinateIndex=1}
  n6 = AppendVector(A=Multiply(A=ComponentMask(None=@n5) {r=true}, B=@n2.TileDetail), B=ComponentMask(None=TextureCoordinate() {coordinateIndex=1}) {g=true})
  n7 = Multiply(A=@n4, B=TextureSample(UVs=Panner(Coordinate=@n6) {speedX=1.5}) {texture=&T_SmashTrail}.G)
  Multiply(A=Multiply(A=LinearInterpolate(A=@n4, B=@n7, Alpha=@n2.LerpDetails), B=@n3.B), B=@n1.A)
 ]
}
